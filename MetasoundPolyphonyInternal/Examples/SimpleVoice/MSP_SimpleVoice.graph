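# MSP_SimpleVoice is an example MetaSound patch implementing the Polyphony.Voice interface as a single mono voice.
# The Start trigger plays a "blip" wave and attacks an ADSR envelope, while Stop releases the envelope.
# The envelope multiplies the wave player's audio to form the output, and its completion fires OnFinished.
# The interface's priority input and output are left unconnected.
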MetaSoundPatch /MetasoundPolyphonyInternal/Examples/SimpleVoice/MSP_SimpleVoice
inputs [
 Polyphony.Voice.In Priority {type=Int32}
 Polyphony.Voice.Start {type=Trigger}
 Polyphony.Voice.Stop {type=Trigger}
]
interfaces [Polyphony.Voice:1.0, UE.OutputFormat.Mono:1.0]
outputs [
 Polyphony.Voice.OnFinished {type=Trigger}
 Polyphony.Voice.Out Priority {type=Int32}
 "UE.OutputFormat.Mono.Audio:0" {type=Audio}
]
pages [Default {
  input_defaults {Polyphony.Voice.In Priority=null, Polyphony.Voice.Start=null, Polyphony.Voice.Stop=null}
  nodes [
   ADSREnvelopeADSREnvelope_0 = ADSR Envelope.ADSR Envelope:Audio(Trigger Attack=^Polyphony.Voice.Start, Trigger Release=^Polyphony.Voice.Stop)
   WavePlayer_0 = UE.Wave Player:Mono(Play=^Polyphony.Voice.Start) {also {defaults {Wave Asset=&/blip}, version [1,1]}}
   Multiply_0 = UE.Multiply:Audio(AdditionalOperands=@ADSREnvelopeADSREnvelope_0.Out Envelope, PrimaryOperand=@WavePlayer_0.Out Mono)
  ]
  links [{dst="@Polyphony.Voice.OnFinished", src=ADSREnvelopeADSREnvelope_0.On Done}, {dst="@UE.OutputFormat.Mono.Audio:0", src=Multiply_0.Out}]
 }]
