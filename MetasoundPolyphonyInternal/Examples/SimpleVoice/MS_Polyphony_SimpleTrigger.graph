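# MS_Polyphony_SimpleTrigger is a mono, looping MetaSound Source example that plays the MSP_SimpleVoice patch polyphonically from repeating triggers.
# On play, two Trigger Repeats with different periods each drive a Trigger Voice Controller.
# The two voice control streams are merged into one Polyphony Patch Player, so the rhythms overlap as a polyrhythm.
# Its voice pool limit is set to the integer maximum, effectively uncapped, and the player's output is the source's audio.
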MetaSoundSource /MetasoundPolyphonyInternal/Examples/SimpleVoice/MS_Polyphony_SimpleTrigger
inputs [UE.Source.OnPlay {type=Trigger}]
interfaces [UE.OutputFormat.Mono:1.0, UE.Source:1.1]
outputs ["UE.OutputFormat.Mono.Audio:0" {type=Audio}]
pages [Default {
  input_defaults {UE.Source.OnPlay=null}
  nodes [
   MetasoundPolyphonyInternalMergeVoiceControlStreams2_0 = "MetasoundPolyphonyInternal.MergeVoiceControlStreams(2)"(
    In 1=MetasoundPolyphonyInternal.TriggerVoiceController(Trigger=UE.TriggerRepeat(Start=^UE.Source.OnPlay) {also {defaults {Period=0.375}}}.RepeatOut) {also {version [0,1]}}.Voice Control Stream
    In 0=MetasoundPolyphonyInternal.TriggerVoiceController(Trigger=UE.TriggerRepeat(Start=^UE.Source.OnPlay) {also {defaults {Period=0.25}}}.RepeatOut) {also {version [0,1]}}.Voice Control Stream
   ) {also {version [0,1]}}
   MetasoundPolyphonyInternalPolyphonyPatchPlayer_0 = MetasoundPolyphonyInternal.PolyphonyPatchPlayer:Mono(
    Voice Control Stream=@MetasoundPolyphonyInternalMergeVoiceControlStreams2_0.Out
    Voice Pool Params=MetasoundPolyphonyInternal.VoicePoolParamsBuilder() {also {defaults {Voice Limit=2147483647}, version [0,1]}}.Out
   ) {also {defaults {MetaSound Asset=&/MSP_SimpleVoice}, version [0,1]}}
  ]
  links [{dst="@UE.OutputFormat.Mono.Audio:0", src=MetasoundPolyphonyInternalPolyphonyPatchPlayer_0.Out Mono}]
 }]
properties {
 assetImportData=&self:AssetImportData
 bLooping=true
 duration=10000
 sampleRate=48000
}
